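# Parent material for the Octo vehicle wheel cosmetic, a masked surface built on a shared simple-alpha wheel function.
# It overrides emissive, world position offset and normal, then passes through vehicle damage and de-rez transition functions.
# Emissive blends fresnel and inner glow with panning bubble textures.
# Sine-driven wind wobble is masked by vertex color and static UV radial shapes, while a switchable tiled detail normal adds rubber grain.
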Material /VehicleCosmetics/Wheels/Octo/Materials/M_WheelParent_Octo
graph {
 properties {
  bAutomaticallySetUsageInEditor=false
  bUseMaterialAttributes=true
  bUsedWithMeshDeformer=true
  bUsedWithSkeletalMesh=true
  blendMode=BLEND_Masked
  pixelDepthOffsetMode=PDOM_Legacy
  usedShadingModels=Default Lit
 }
 nodes [
  n1 = GetMaterialAttributes(None=MaterialFunctionCall() {materialFunction=/VehicleCosmetics/Wheels/Ultrabarro_Radiant/Materials/MF_WheelParent_Simple_Alpha}.MaterialAttributes) {attributeGetTypes [#1]}
  n2 = NamedRerouteUsage()
  n3 = NamedRerouteUsage()
  n4 = NamedRerouteUsage()
  MaterialFunctionCall(In=MaterialFunctionCall(Material=SetMaterialAttributes!(MaterialAttributes=@n1.MaterialAttributes, Emissive Color=@n2, World Position Offset=@n3, Normal=@n4) {attributeSetTypes [#2, #3, #1]}) {materialFunction=&MF_BaseVehicle_Damage}.Result) {materialFunction=&MF_VehicleDeRez_Transition}
  n5 = Add(A=TextureCoordinate() {uTiling=2, vTiling=2}, B=Constant() {r=-1})
  n6 = Constant() {r=1}
  n7 = Panner(Coordinate=@n5, Time=@n6) {speedX=-0.75, speedY=0.42}
  n8 = DotProduct(A=@n7, B=@n7)
  n9 = Panner(Coordinate=@n5, Time=@n6) {speedX=-0.46, speedY=0.635}
  n10 = DotProduct(A=@n9, B=@n9)
  n11 = Multiply(A=Clamp(None=ConstantBiasScale(None=@n8) {bias=-0.02, scale=60}), B=Clamp(None=ConstantBiasScale(None=@n10) {bias=-0.02, scale=60}))
  n12 = VectorParameter() {group=Octo, parameterName=WindDirectionXYZ-RGB, defaultValue=<rgba 1 1 0 1>}
  n13 = Rotator(Time=Constant() {r=-0.5}) {speed=3.14}
  n14 = ComponentMask(None=@n13) {r=true}
  n15 = Multiply(A=Time(), B=Multiply(A=ScalarParameter() {defaultValue=1.5, group=Octo, parameterName=WindSpeed}, B=Constant() {r=1}))
  n16 = Sine(None=Add(A=OneMinus(None=@n14), B=@n15)) {period=2}
  n17 = Sine(None=Add(A=ComponentMask(None=@n13) {g=true}, B=@n15)) {period=4}
  n18 = Add(A=@n16, B=@n17)
  n19 = Multiply(A=Multiply(A=Add(A=@n16, B=@n18), B=Add(A=@n17, B=@n18)), B=VertexColor().A)
  n20 = ScalarParameter() {defaultValue=0.5, group=Octo, parameterName=WindIntensity}
  n21 = Multiply(
   A=Multiply(A=Transform(None=@n12.RGB), B=Multiply(A=@n19, B=@n20))
   B=ScalarParameter() {defaultValue=0.1, group=Octo, parameterName=Animation}
  )
  NamedRerouteDeclaration(None=LinearInterpolate(A=Multiply(A=@n11, B=@n21), B=Constant(), Alpha=VertexColor().R)) {name=WPO, nodeColor=<rgba 0.2111 1 0.6548 1>}
  n22 = TextureSampleParameter2D() {parameterName=Mask_VFX, texture=/VehicleCosmetics/Wheels/Octo/Textures/T_Octo_Rim_RGB}
  NamedRerouteDeclaration(None=@n22.B) {name=RGB_Blue, nodeColor=<rgba 0 1 0.7188 1>}
  NamedRerouteDeclaration(None=@n22.G) {name=RGB_green, nodeColor=<rgba 0.2422 1 0 1>}
  NamedRerouteDeclaration(None=@n22.R) {name=RGB_red, nodeColor=<rgba 1 0.0237 0 1>}
  NamedRerouteDeclaration(None=@n1.Normal) {name=Normal_IN, nodeColor=<rgba 0 0.8828 1 1>}
  n23 = NamedRerouteUsage()
  n24 = TextureCoordinate()
  n25 = ScalarParameter() {defaultValue=2, group=Octo, parameterName=TextureDetailNormalTiling}
  n26 = Multiply(
   A=TextureSample(UVs=Multiply(A=@n24, B=@n25)) {samplerType=SAMPLERTYPE_Normal, texture=/VehicleCosmetics/SharedTextures/T_Matte_N}.RGB
   B=TextureSampleParameter2D() {parameterName=Mask_VC, samplerType=SAMPLERTYPE_LinearColor, texture=/VehicleCosmetics/Wheels/Octo/Textures/T_Octo_Rim_VC}.A
  )
  NamedRerouteDeclaration(None=StaticSwitchParameter(True=Add(A=@n23, B=@n26), False=@n23) {defaultValue=true, group=TireRubber, parameterName=TireRubberExist}) {name=NormalRubber, nodeColor=<rgba 0.1016 0 1 1>}
  NamedRerouteDeclaration(None=@n22.A) {name=RGB_A, nodeColor=<rgba 0.9101 1 0.9875 1>}
  n27 = VectorParameter() {parameterName=FresnelColor, defaultValue=<rgba 0 0.7693 1 1>}
  n28 = Multiply(A=Multiply(A=NamedRerouteUsage(), B=@n27.RGB), B=ScalarParameter() {defaultValue=1, group=Octo, parameterName=InsideLight})
  n29 = Multiply(
   A=@n27.RGB
   B=MaterialFunctionCall!(
    Input Normal=NamedRerouteUsage()
    Input Sharpness=ScalarParameter() {defaultValue=3, group=Octo, parameterName=FresnelSharpness}
    Input Two Sided
   ) {materialFunction=/VehicleCosmetics/Wheels/CarCar_Holographic/Materials/MF_Fresnel_Advanced}.Edge
  )
  n30 = ScalarParameter() {defaultValue=0.2, group=Octo, parameterName=FresnelBrightness}
  n31 = Constant()
  n32 = LinearInterpolate(
   A=LinearInterpolate(
    A=Multiply(A=@n29, B=@n30)
    B=@n31
    Alpha=TextureSampleParameter2D() {parameterName=Mask_VC, samplerType=SAMPLERTYPE_LinearColor, texture=&T_Octo_Rim_VC}.B
   )
   B=@n31
   Alpha=VertexColor().R
  )
  n33 = ScalarParameter() {defaultValue=1, group=Octo, parameterName=EmissiveBubbles}
  n34 = TextureSample(UVs=Panner(Coordinate=TextureCoordinate() {uTiling=1.8, vTiling=2}) {speedX=0.3, speedY=-0.01}) {texture=/VehicleCosmetics/Wheels/Octo/Textures/Bubbles}
  n35 = TextureSample(UVs=Panner(Coordinate=TextureCoordinate() {uTiling=2, unMirrorU=true, unMirrorV=true, vTiling=3}) {speedX=0.35, speedY=0.1}) {texture=&Bubbles}
  n36 = ComponentMask(None=Rotator(Time=Constant() {r=-0.47}) {speed=3.14}) {r=true}
  n37 = OneMinus(None=ConstantBiasScale(None=OneMinus(None=@n36)) {bias=-0.11, scale=200})
  n38 = NamedRerouteUsage()
  n39 = Add(
   A=Multiply(A=LinearInterpolate(A=@n34.RGB, B=@n35.RGB, Alpha=@n35.RGB), B=Multiply(A=@n37, B=@n38))
   B=VectorParameter() {parameterName=CustomColorBubbles, defaultValue=<rgba 0 0.3286 0.4271 1>}.RGB
  )
  NamedRerouteDeclaration(None=LinearInterpolate(A=Add(A=@n28, B=@n32), B=Multiply(A=@n33, B=@n39), Alpha=NamedRerouteUsage())) {name=Emission, nodeColor=<rgba 0.8281 0 1 1>}
 ]
}
